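# Default skin material for the Quartz vehicle body: a Default Lit, metallic-by-default surface for skeletal meshes.
# A parameter-scaled Fresnel blends Color toward AccentColor for base color, with the scale defaulting to zero.
# Scalar parameters drive metallic and roughness, and BodyNormalLOD supplies the normal, defaulting to a blank map.
# The material attributes pass through MF_BaseVehicle_Damage, then MF_VehicleDeRez_Transition.
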Material /VehicleCosmetics/Bodies/Quartz/Skins/M_Skin_QuartzDefault
graph {
 properties {
  bAutomaticallySetUsageInEditor=false
  bUseMaterialAttributes=true
  bUsedWithMeshDeformer=true
  bUsedWithSkeletalMesh=true
  usedShadingModels=Default Lit
 }
 nodes [
  n1 = Fresnel(
   ExponentIn=ScalarParameter() {defaultValue=5, parameterName=MultiFresnal_2}
   BaseReflectFractionIn=ScalarParameter() {defaultValue=0.04, parameterName=MultiFresnal_1}
  ) {baseReflectFraction=0.02}
  n2 = ScalarParameter() {parameterName=MultiFresnal}
  n3 = LinearInterpolate(
   A=VectorParameter() {parameterName=Color, defaultValue=<rgba 0.5136 0.5312 0.4965 1>}.RGB
   B=VectorParameter() {parameterName=AccentColor, defaultValue=<rgba 0.3916 0.3916 0.3916 1>}.RGB
   Alpha=Saturate(None=Multiply(A=@n1, B=@n2))
  )
  n4 = ScalarParameter() {defaultValue=1, parameterName=Metal}
  n5 = ScalarParameter() {defaultValue=0.22, parameterName=Roughness}
  n6 = TextureSampleParameter2D() {parameterName=BodyNormalLOD, samplerType=SAMPLERTYPE_Normal, texture=/VehicleCosmetics/Textures/T_BlankNormal_B}
  MaterialFunctionCall(In=MaterialFunctionCall(Material=MakeMaterialAttributes!(
     BaseColor=@n3
     Metallic=@n4
     Specular
     Roughness=@n5
     Anisotropy
     EmissiveColor
     Opacity
     OpacityMask
     Normal=@n6.RGB
     Tangent
     WorldPositionOffset
     SubsurfaceColor
     ClearCoat
     ClearCoatRoughness
     AmbientOcclusion
     Refraction
     CustomizedUVs_0
     CustomizedUVs_1
     CustomizedUVs_2
     CustomizedUVs_3
     CustomizedUVs_4
     CustomizedUVs_5
     CustomizedUVs_6
     CustomizedUVs_7
     PixelDepthOffset
     ShadingModel
     Displacement
    )) {materialFunction=&MF_BaseVehicle_Damage}.Result) {materialFunction=&MF_VehicleDeRez_Transition}
 ]
}
